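# Physics asset for the medium Meatball drivable vehicle, defining collision shapes named for the driver and passenger positions.
# A single root body holds four capsules: driver at the front, a middle passenger near the center, and two rear passengers offset left and right.
# There are no constraints and no shapes for the hull itself, so this asset covers only those occupant volumes.
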PhysicsAsset /Game/Athena/DrivableVehicles/Meatball/Meshes/Medium_Meatball_Veh_Physics
bodies {root [
  {shapeType=Capsule, length=45, radius=30, shapeName=DriverCollision, center=<78.0664 -0.0001 120.2578>}
  {shapeType=Capsule, length=60, radius=35, shapeName=PassengerBLCollision, center=<-86.9336 -50.0001 125.2578>}
  {shapeType=Capsule, length=60, radius=35, shapeName=PassengerBRCollision, center=<-86.9336 49.9999 125.2578>}
  {shapeType=Capsule, length=80, radius=35, shapeName=PassengerMidCollision, center=<-1.9336 -0.0001 120.2578>}
 ]}
constraints []
