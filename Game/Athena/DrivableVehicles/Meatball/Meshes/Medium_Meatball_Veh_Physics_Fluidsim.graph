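# Physics asset for the medium Meatball drivable vehicle, named as the variant used for fluid simulation.
# A single root body approximates the vehicle's shape with five capsules: one large central capsule, two long side capsules and two smaller angled capsules toward +X.
# The side and forward pairs are roughly mirrored across the centerline, and no constraints are defined, so the shape acts as one rigid piece.
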PhysicsAsset /Game/Athena/DrivableVehicles/Meatball/Meshes/Medium_Meatball_Veh_Physics_Fluidsim
bodies {root [
  {
   shapeType=Capsule
   length=112.5
   radius=38
   shapeName=root_capsule
   center=<247.9867 -61.5773 51.1289>
   rotation=<rot -81.3955 5.6552 4.1492>
  }
  {
   shapeType=Capsule
   length=112.5
   radius=38
   shapeName=root_capsule
   center=<248.0657 59.9957 51.0535>
   rotation=<rot -81.3956 -8.4552 -3.0421>
  }
  {shapeType=Capsule, length=159.5, radius=103.5, shapeName=root_capsule, center=<33.9292 0 105.8859>, rotation=<rot -90 90 -90.0001>}
  {
   shapeType=Capsule
   length=171.5
   radius=52
   shapeName=root_capsule
   center=<86.285 66.7599 67.1858>
   rotation=<rot 89.0895 -179.9886 -179.9886>
  }
  {
   shapeType=Capsule
   length=171.5
   radius=52
   shapeName=root_capsule
   center=<86.285 -68.0284 67.186>
   rotation=<rot 89.0895 -179.9886 -179.9886>
  }
 ]}
constraints []
